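# Unlit, two-sided test material for thickened spline meshes, also flagged for Nanite and mesh particle use.
# Thresholded sine and cosine of the UVs build a four-shade quadrant pattern, which is multiplied by vertex color and passed through emissive adjustments.
# SplineThicken receives one animated width for base and tip, oscillating between Width Low and Width via DebugTimeSine, with a static switch choosing U or V expansion.
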Material /Game/Effects/Fort_Effects/Mesh/Splines/M_SplineTestMaterial
graph {
 properties {
  bUsedWithMeshParticles=true
  bUsedWithNanite=true
  bUsedWithNiagaraMeshParticles=true
  bUsedWithSplineMeshes=true
  pixelDepthOffsetMode=PDOM_Legacy
  shadingModel=MSM_Unlit
  twoSided=true
  usedShadingModels=Unlit
 }
 nodes [
  n1 = TextureCoordinate()
  n2 = Ceil(None=Sine(None=ComponentMask(None=@n1) {r=true}))
  n3 = Ceil(None=Cosine(None=ComponentMask(None=@n1) {g=true}) {period=2})
  n4 = OneMinus(None=@n3)
  n5 = OneMinus(None=@n2)
  n6 = Add(A=Subtract(A=@n2, B=@n4), B=Multiply(A=Subtract(A=@n5, B=@n4)) {constB=0.75})
  n7 = Add(A=Multiply(A=Subtract(A=@n2, B=@n3)) {constB=0.5}, B=Multiply(A=Subtract(A=@n5, B=@n3)) {constB=0.25})
  MaterialFunctionCall(Emissive=Multiply(A=LinearInterpolate(A=@n6, B=@n7, Alpha=@n4), B=VertexColor())) {materialFunction=&MF_FX_EmissiveAdjustments}
  n8 = LinearInterpolate(
   A=ScalarParameter() {parameterName=Width Low}
   B=ScalarParameter() {defaultValue=10, parameterName=Width}
   Alpha=MaterialFunctionCall() {materialFunction=&DebugTimeSine}.Result
  ) {constA=10}
  MaterialFunctionCall(WidthBase=@n8, WidthTip=@n8, Expand U or V UV Channel=StaticBoolParameter() {parameterName="Expand U or V UV Channel (SB)"}) {materialFunction=&SplineThicken}
 ]
}
